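# Translucent, surface-lit plastic material for the Aventurine01 vehicle wheel, enabled for skeletal meshes and mesh deformers.
# Named reroutes share Fresnel, vertex-color red, mask red, base and edge colors between sections that blend color, roughness, metallic and opacity and scale emission.
# Static switches select the custom mask and choose the wheel's normal map or an alternative input.
# Side chains labeled "Should be Final Color" and "Should be Final Normal" hold alternative color and normal math.
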Material /VehicleCosmetics/Wheels/Aventurine01/Materials/M_Wheel_Aventurine01_Plastic
graph {
 properties {
  bAutomaticallySetUsageInEditor=false
  bUsedWithMeshDeformer=true
  bUsedWithSkeletalMesh=true
  blendMode=BLEND_Translucent
  translucencyLightingMode=TLM_Surface
  usedShadingModels=Default Lit
 }
 nodes [
  LinearInterpolate(
   A=ScalarParameter() {defaultValue=0.2, group=0 - Base, parameterName=Roughness, sortPriority=5}
   B=ScalarParameter() {defaultValue=0.5, group=0 - Base, parameterName=RoughnessEdges, sortPriority=6}
   Alpha=NamedRerouteUsage()
  )
  n1 = Multiply(A=NamedRerouteUsage(), B=NamedRerouteUsage())
  n2 = NamedRerouteUsage()
  n3 = Add(A=NamedRerouteUsage(), B=@n2) {constB=0.2}
  n4 = ScalarParameter() {defaultValue=2.5, group=1 - Fresnel, parameterName=Fresnel Edges Contrast, sortPriority=2}
  n5 = Clamp(None=Multiply(A=Power(Base=@n3, Exp=@n4), B=Add(A=@n2) {constB=0.2}))
  n6 = TextureSample() {samplerType=SAMPLERTYPE_Normal, texture=/VehicleCosmetics/Textures/T_BlankNormal_B}
  n7 = CameraVectorWS()
  n8 = OneMinus(None=Clamp(None=DotProduct(A=@n6.R, B=@n7)))
  n9 = Multiply(A=Add(A=Multiply(A=@n8, B=@n8)) {constB=0.25}, B=NamedRerouteUsage())
  LinearInterpolate(A=Add(A=@n1, B=Add(A=@n5, B=@n9)), B=NamedRerouteUsage(), Alpha=OneMinus(None=@n1)) {desc=Should be Final Color}
  Multiply(A=@n6.RGB, B=Constant3Vector() {constant=<rgba 0.75 0.75 1.2 0>}) {desc=Should be Final Normal}
  NamedRerouteDeclaration(None=Fresnel(
    ExponentIn=ScalarParameter() {defaultValue=5, group=1 - Fresnel, parameterName=Fresnel Falloff, sortPriority=0}
    BaseReflectFractionIn=ScalarParameter() {defaultValue=0.04, group=1 - Fresnel, parameterName=Fresnel Reflect Fraction, sortPriority=1}
   )) {name=Fresnel, nodeColor=<rgba 0 0.6667 1 1>}
  NamedRerouteDeclaration(None=VectorParameter() {group=0 - Base, parameterName=Edge Color, sortPriority=0, defaultValue=<rgba 0 0.3712 0.3712 1>}.RGB) {name=Edge Color, nodeColor=<rgba 1 0.5 0 1>}
  NamedRerouteDeclaration(None=VertexColor().R) {name=VC_R, nodeColor=<rgba 0 1 0.7422 1>}
  n10 = TextureSampleParameter2D() {
   group=0 - Base
   parameterName=Normal
   samplerSource=SSM_Wrap_WorldGroupSettings
   samplerType=SAMPLERTYPE_Normal
   sortPriority=2
   texture=/VehicleCosmetics/Wheels/Aventurine01/Textures/T_Aventurine01_Plastic_N
  }
  NamedRerouteDeclaration(None=StaticSwitchParameter(True=@n10.RGB, False=NamedRerouteUsage()) {defaultValue=true, group=0 - Base, parameterName=Custom Normal, sortPriority=1}) {name=Normal, nodeColor=<rgba 0.5 0.5 1 1>}
  n11 = TextureSampleParameter2D() {group=0 - Base, parameterName=Custom Mask, sortPriority=2, texture=/VehicleCosmetics/Wheels/Aventurine01/Textures/T_Aventurine01_Plastic_M}
  NamedRerouteDeclaration(None=@n11.R) {name=CustomMask_R, nodeColor=<rgba 0 1 0.7422 1>}
  NamedRerouteDeclaration(None=VectorParameter() {group=0 - Base, parameterName=Color, sortPriority=0, defaultValue=<rgba 0 0.0417 0.0417 1>}.RGB) {name=Color, nodeColor=<rgba 1 0.5 0 1>}
  n12 = NamedRerouteUsage()
  n13 = NamedRerouteUsage()
  n14 = Multiply(
   A=Clamp(None=Multiply(A=@n12, B=@n13))
   B=StaticSwitchParameter(True=NamedRerouteUsage(), False=NamedRerouteUsage()) {group=0 - Base, parameterName=Custom Mask, sortPriority=1}
  )
  NamedRerouteDeclaration(None=LinearInterpolate(A=@n14, B=NamedRerouteUsage(), Alpha=OneMinus(None=@n14))) {name=Final Color, nodeColor=<rgba 1 0.5 0 1>}
  NamedRerouteDeclaration(None=TextureSample() {samplerSource=SSM_Wrap_WorldGroupSettings, samplerType=SAMPLERTYPE_Normal, texture=&T_BlankNormal_B}.RGB) {name=Blank Final Normal, nodeColor=<rgba 0.5029 0.5029 1 1>}
  NamedRerouteUsage()
  NamedRerouteUsage()
  Reroute(None=LinearInterpolate(
    A=ScalarParameter() {group=0 - Base, parameterName=Metallic, sortPriority=3}
    B=ScalarParameter() {defaultValue=1, group=0 - Base, parameterName=MetallicEdges, sortPriority=4}
    Alpha=NamedRerouteUsage()
   ))
  n15 = NamedRerouteUsage()
  n16 = StaticSwitchParameter(True=NamedRerouteUsage(), False=NamedRerouteUsage()) {group=0 - Base, parameterName=Custom Mask, sortPriority=1}
  Reroute(None=Multiply(A=Multiply(A=@n15, B=@n16), B=ScalarParameter() {defaultValue=1, group=0 - Base, parameterName=Emission Int, sortPriority=7}))
  n17 = NamedRerouteUsage()
  n18 = Multiply(A=@n16, B=ScalarParameter() {defaultValue=1, group=0 - Base, parameterName=OpacityEdges, sortPriority=9})
  Reroute(None=LinearInterpolate(A=ScalarParameter() {group=0 - Base, parameterName=Opacity, sortPriority=8}, B=Add(A=@n18, B=@n17), Alpha=@n17))
 ]
}
